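; Blueprint interface through which music encounters are notified of boss state changes.
; It declares callbacks for alert level, conversion, DBNO, death, health and target changes, plus whether the boss is targeting the listener.
; Each callback passes the boss component as an interface reference, alongside any event data such as the health info struct, converted pawn or targeted actor.
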(fn BossAlertLevelChanged
  (params (Boss%20Component%20Interface :type Music%20Encounter%20Component%20Boss%20Interface%20Interface) (Alert%20Level :type EAlertLevel%20Enum :default "Unaware")))

(fn BossConversionStateChanged
  (params (BossComponentInterface :type Music%20Encounter%20Component%20Boss%20Interface%20Interface) (Instigator :type Fort%20Pawn%20Object%20Reference) (Converted%20Pawn :type Fort%20Pawn%20Object%20Reference) (Converted :type Boolean)))

(fn BossDBNOChanged
  (params (Boss%20Component%20Interface :type Music%20Encounter%20Component%20Boss%20Interface%20Interface) (Is%20DBNO :type Boolean)))

(fn BossDeath
  (params (Boss%20Component%20Interface :type Music%20Encounter%20Component%20Boss%20Interface%20Interface)))

(fn BossHealthChanged
  (params (Boss%20Component%20Interface :type Music%20Encounter%20Component%20Boss%20Interface%20Interface) (Health%20Info :type Struct%20NPC%20Health%20Info%20Structure)))

(fn BossTargetChanged
  (params (Boss%20Component%20Interface :type Music%20Encounter%20Component%20Boss%20Interface%20Interface) (Targeted%20Actor :type Actor%20Object%20Reference)))

(fn IsBossTargetingListenerChanged
  (params (Boss%20Component%20Interface :type Music%20Encounter%20Component%20Boss%20Interface%20Interface) (Is%20Boss%20Targeting%20Listener :type Boolean)))
