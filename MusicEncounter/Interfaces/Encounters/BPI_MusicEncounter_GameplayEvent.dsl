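; Blueprint interface for gameplay events in music encounters, declaring four ways to execute an event on an implementer.
; Each function receives a Music Encounter Component Base reference plus a payload identifying the event: a byte enum, Gameplay Tag, Name or Instanced Struct.
; It holds only signatures with no return values, so implementing objects define what each event does.
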(fn ExecuteEnumEvent
  (params (Music%20Encounter%20Component :type Music%20Encounter%20Component%20Base%20Object%20Reference) (Event%20Enum :type Byte)))

(fn ExecuteGameplayTagEvent
  (params (Music%20Encounter%20Component :type Music%20Encounter%20Component%20Base%20Object%20Reference) (Gameplay%20Tag :type Gameplay%20Tag%20Structure)))

(fn ExecuteNamedEvent
  (params (Music%20Encounter%20Component :type Music%20Encounter%20Component%20Base%20Object%20Reference) (Event%20Name :type Name)))

(fn ExecuteStructEvent
  (params (Music%20Encounter%20Component :type Music%20Encounter%20Component%20Base%20Object%20Reference) (Event%20Struct :type Instanced%20Struct%20Structure)))
