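; Blueprint interface that passes NPC state changes to music encounter objects.
; It declares four output-less functions for alert level changes, NPC DBNO changes, NPC death and target changes.
; The DBNO and death calls carry the non-participant player pawn, while the alert call carries an alert level enum defaulting to Unaware and the target call carries an actor.
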(fn AlertLevelChanged
  (params (Alert%20Level :type EAlertLevel%20Enum :default "Unaware")))

(fn NPCDBNOChanged
  (params (NPC%20Pawn :type BP%20Player%20Pawn%20Non%20Participant%20Object%20Reference) (Is%20DBNO :type Boolean)))

(fn NPCDeath
  (params (NPC%20Pawn :type BP%20Player%20Pawn%20Non%20Participant%20Object%20Reference)))

(fn TargetChanged
  (params (Target%20Actor :type Actor%20Object%20Reference)))
