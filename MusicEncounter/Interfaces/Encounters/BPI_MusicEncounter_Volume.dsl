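; Blueprint interface for music encounter volumes that reports a listener moving into or out of a volume.
; It declares two functions, ListenerEntered and ListenerExited, with no parameters or return values shown.
; The file holds only these declarations, so the response to each call is left to the implementing classes.
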(fn ListenerEntered)

(fn ListenerExited)
